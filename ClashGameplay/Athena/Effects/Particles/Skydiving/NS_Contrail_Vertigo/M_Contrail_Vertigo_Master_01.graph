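# Translucent unlit, two-sided master material for Vertigo skydiving contrail Niagara effects, with parameter blends between a goop look and a bubbly liquid trail.
# Panning noise, distorted mask and gradient textures form clipped, edge-masked opacity with camera and depth fades.
# They also drive colors plus a fake matcap highlight, optional sun/view scattering and time-of-day glow, simplified on low-quality and mobile.
# Dynamic parameters control random offset, clipping and noise-based world position offset.
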Material /ClashGameplay/Athena/Effects/Particles/Skydiving/NS_Contrail_Vertigo/M_Contrail_Vertigo_Master_01
graph {
 properties {
  bAutomaticallySetUsageInEditor=false
  bUsedWithNiagaraMeshParticles=true
  bUsedWithNiagaraRibbons=true
  bUsedWithNiagaraSprites=true
  bUsedWithStaticMesh=false
  blendMode=BLEND_Translucent
  pixelDepthOffsetMode=PDOM_Legacy
  shadingModel=MSM_Unlit
  twoSided=true
  usedShadingModels=Unlit
 }
 nodes [
  n1 = LinearInterpolate(
   A=TextureCoordinate()
   B=TextureCoordinate() {coordinateIndex=1}
   Alpha=ScalarParameter() {defaultValue=1, group=Controls, parameterName=Use UV1}
  )
  n2 = ComponentMask(None=VectorParameter() {group=Noise, parameterName=Noise UV Scale, defaultValue=<rgba 1 0.1 0 0>}.RGB) {g=true, r=true}
  n3 = Time() {bOverride_Period=true, period=64}
  n4 = DynamicParameter() {defaultValue=<rgba 1 1 0 1>, paramNames [rand offset, ClipAmnt, WPO Intensity, Param4]}
  n5 = VectorParameter() {group=Noise, parameterName=Noise Panner Speed, defaultValue=<rgba -0.8 0 0 0>}
  n6 = TextureSampleParameter2D(UVs=Panner(Coordinate=Multiply(A=@n1, B=@n2), Time=Add(A=@n3, B=@n4.rand offset), Speed=ComponentMask(None=@n5.RGB) {g=true, r=true}) {speedX=-0.1, speedY=-0.01}) {
   group=Noise
   parameterName=Noise Texture
   samplerSource=SSM_Wrap_WorldGroupSettings
   texture=/Game/Global/Textures/Default/Caustics/T_Fractal_01
  }
  n7 = ChannelMaskParameter(None=@n6.RGBA) {group=Noise, maskChannel=Green, parameterName=Noise Channel Mask, defaultValue=<rgba 0 1 0 0>}
  n8 = LinearInterpolate(
   A=TextureCoordinate()
   B=TextureCoordinate() {coordinateIndex=1}
   Alpha=ScalarParameter() {defaultValue=1, group=Controls, parameterName=Use UV1}
  )
  n9 = ComponentMask(None=VectorParameter() {group=Mask, parameterName=Mask UV Scale, defaultValue=<rgba 4 1 0 0>}.RGB) {g=true, r=true}
  n10 = LinearInterpolate(
   A=TextureCoordinate()
   B=TextureCoordinate() {coordinateIndex=1}
   Alpha=ScalarParameter() {defaultValue=1, group=Controls, parameterName=Use UV1}
  )
  n11 = ComponentMask(None=VectorParameter() {group=Distortion, parameterName=Distortion Tile, defaultValue=<rgba 6 1 0 0>}.RGB) {g=true, r=true}
  n12 = Time() {bOverride_Period=true, period=64}
  n13 = DynamicParameter() {defaultValue=<rgba 1 1 0 1>, paramNames [rand offset, ClipAmnt, WPO Intensity, Param4]}
  n14 = VectorParameter() {group=Distortion, parameterName=Distortion Panner Speed, defaultValue=<rgba -4.5 0.1 0 0>}
  n15 = TextureSampleParameter2D(UVs=Panner(Coordinate=Multiply(A=@n10, B=@n11), Time=Add(A=@n12, B=@n13.rand offset), Speed=ComponentMask(None=@n14.RGB) {g=true, r=true}) {speedX=-1, speedY=0.1}) {
   group=Distortion
   parameterName=Distortion Texture
   samplerSource=SSM_Wrap_WorldGroupSettings
   samplerType=SAMPLERTYPE_Grayscale
   texture=&LowResBlurredNoise
  }
  n16 = Multiply(
   A=LinearInterpolate(Alpha=ChannelMaskParameter(None=@n15.RGBA) {group=Distortion, maskChannel=Green, parameterName=Distortion Mask, defaultValue=<rgba 0 1 0 0>}.R) {constA=-1}
   B=ScalarParameter() {defaultValue=0.15, group=Distortion, parameterName=DistAmnt}
  )
  n17 = Panner(
   Coordinate=Add(A=Multiply(A=@n8, B=@n9), B=@n16)
   Time=Add(
    A=Time() {bOverride_Period=true, period=64}
    B=DynamicParameter() {defaultValue=<rgba 1 1 0 1>, paramNames [rand offset, ClipAmnt, WPO Intensity, Param4]}.rand offset
   )
   Speed=ComponentMask(None=VectorParameter() {group=Mask, parameterName=Mask Panner Speed, defaultValue=<rgba -2.6 0 0 0>}.RGB) {g=true, r=true}
  ) {speedX=-0.8}
  n18 = ComponentMask(None=VectorParameter() {group=Mask, parameterName=Mask Texture Offset}.RGB) {g=true, r=true}
  n19 = ChannelMaskParameter(None=TextureSampleParameter2D(UVs=Add(A=@n17, B=@n18)) {group=Mask, parameterName=Mask Texture, texture=/Game/Effects/Fort_Effects/Textures/Smoke/T_Smoke_Trail_Soft_Chromatic}.RGBA) {group=Mask, maskChannel=Green, parameterName=Mask Channel Mask, defaultValue=<rgba 0 1 0 0>}
  n20 = TextureCoordinate()
  n21 = TextureCoordinate() {coordinateIndex=1}
  n22 = ScalarParameter() {defaultValue=1, parameterName=Use UV1}
  n23 = Add(
   A=Add(A=@n16, B=LinearInterpolate(A=@n20, B=@n21, Alpha=@n22))
   B=ComponentMask(None=VectorParameter() {group=Alpha Mask, parameterName=Alpha Mask UV Offset}.RGB) {g=true, r=true}
  )
  n24 = MaterialFunctionCall(Float2=@n23) {materialFunction=&BreakOutFloat2Components}
  n25 = OneMinus(None=@n24.R)
  n26 = OneMinus(None=@n24.G)
  n27 = TextureSampleParameter2D(UVs=LinearInterpolate(A=@n23, B=AppendVector(A=@n25, B=@n26), Alpha=ScalarParameter() {group=Alpha Mask, parameterName=Invert Alpha Texture UVs})) {group=Alpha Mask, parameterName=Alpha Texture, texture=/Game/Global/Textures/Default/Masks_and_Gradients/gradient_stasis}
  n28 = Add(A=Multiply(A=@n7.R, B=@n19.R), B=@n27.G)
  n29 = Multiply(
   A=MaterialFunctionCall(In=@n7.R, Contrast=ScalarParameter() {defaultValue=0.15, group=Noise, parameterName=Noise Contrast}) {materialFunction=&CheapContrast}.Result
   B=MaterialFunctionCall(In=@n19.R, Contrast=ScalarParameter() {group=Mask, parameterName=Mask Contrast}) {materialFunction=&CheapContrast}.Result
  )
  n30 = LinearInterpolate(
   A=TextureCoordinate()
   B=TextureCoordinate() {coordinateIndex=1}
   Alpha=ScalarParameter() {defaultValue=1, group=Controls, parameterName=Use UV1}
  )
  n31 = ComponentMask(None=VectorParameter() {group=Cube Highlight, parameterName=Noise Normal UV Scale, defaultValue=<rgba 1 1 0 0>}.RGB) {g=true, r=true}
  n32 = Time() {bOverride_Period=true, period=64}
  n33 = DynamicParameter() {defaultValue=<rgba 1 1 0 1>, paramNames [rand offset, ClipAmnt, WPO Intensity, Param4]}
  n34 = VectorParameter() {group=Cube Highlight, parameterName=Noise Normal Panner Speed, defaultValue=<rgba -0.3 -0.01 0 0>}
  n35 = TextureSampleParameter2D(UVs=Panner(Coordinate=Multiply(A=@n30, B=@n31), Time=Add(A=@n32, B=@n33.rand offset), Speed=ComponentMask(None=@n34.RGB) {g=true, r=true}) {speedX=-0.3, speedY=-0.01}) {
   group=Cube Highlight
   parameterName=Noise Normal Texture
   samplerSource=SSM_Wrap_WorldGroupSettings
   samplerType=SAMPLERTYPE_Normal
   texture=/Game/Global/Textures/Default/Tileable_Patterns/2ndInfectedMapNormal
  }
  n36 = Multiply(A=@n35.RGB, B=ScalarParameter() {defaultValue=-3, group=Cube Highlight, parameterName=Noise Normal Intensity})
  n37 = Multiply(
   A=Add(A=LinearInterpolate(Alpha=@n28), B=Multiply(A=@n29, B=@n36))
   B=ScalarParameter() {defaultValue=0.2, group=Cube Highlight, parameterName=Cubemap Distortion Amount}
  )
  n38 = TextureCoordinate()
  n39 = TextureCoordinate() {coordinateIndex=1}
  n40 = ScalarParameter() {defaultValue=1, group=Controls, parameterName=Use UV1}
  n41 = Add(A=Add(A=ComponentMask(None=@n37) {g=true, r=true}, B=LinearInterpolate(A=@n38, B=@n39, Alpha=@n40))) {constB=-0.5}
  n42 = TextureSample(UVs=Add(A=Multiply(A=@n41) {constB=0.8}) {constB=0.5}) {
   samplerSource=SSM_Wrap_WorldGroupSettings
   samplerType=SAMPLERTYPE_Normal
   texture=&T_Sphere_N
  }
  n43 = AppendVector(A=AppendVector(A=@n42.R, B=OneMinus(None=@n42.G)), B=@n42.B)
  n44 = Add(A=Multiply(A=Transform(None=@n43) {transformType=TRANSFORM_View}) {constB=0.5}) {constB=0.5}
  n45 = MaterialFunctionCall!(
   In=TextureSample(UVs=ComponentMask(None=@n44) {g=true, r=true}) {samplerSource=SSM_Wrap_WorldGroupSettings, texture=/Game/Packages/Fortress_SharedMaterials/Textures/T_Ravens_SpecularReflection}.RGB
   Contrast=ScalarParameter() {defaultValue=0.3, group=Cube Highlight, parameterName=CubeContrast}
  ) {materialFunction=/Engine/Functions/Engine_MaterialFunctions01/ImageAdjustment/CheapContrast_RGB}
  n46 = Constant3Vector() {constant=<rgba 1 1 1 0>}
  n47 = VectorParameter() {group=Liquid Color Scattering, parameterName=Liquid_InnerBackScatteringMultiplier, defaultValue=<rgba 0 0 0 1>}
  n48 = CollectionParameter() {
   collection=&FortniteMaterialParameters
   parameterName=SunlightDirectionalVector
  }
  n49 = ComponentMask(None=@n48) {b=true, g=true, r=true}
  n50 = VertexNormalWS()
  n51 = Multiply(A=Multiply(A=DotProduct(A=@n49, B=@n50)) {constB=2}, B=@n50)
  n52 = DotProduct(
   A=Normalize(VectorInput=Subtract(A=@n51, B=@n49))
   B=Multiply(A=MaterialFunctionCall() {materialFunction=&CameraDirectionVector}.Result) {constB=-1}
  )
  n53 = Max(A=Add(A=Multiply(A=@n52) {constB=-1}, B=ScalarParameter() {group=Liquid Color Scattering, parameterName=Liquid_InnerBackScatteringBias})) {constB=0}
  n54 = ScalarParameter() {defaultValue=1, group=Liquid Color Scattering, parameterName=Liquid_InnerBackScatteringPower}
  n55 = Saturate(None=Multiply(A=Power(Base=@n53, Exp=@n54), B=ScalarParameter() {group=Liquid Color Scattering, parameterName=Liquid_InnerBackScatteringScale}) {constB=0.55})
  n56 = VectorParameter() {group=Liquid Color Scattering, parameterName=Liquid_OuterColorA, defaultValue=<rgba 0.0125 0.724 0.6663 1>}
  n57 = TransformPosition(None=WorldPosition().XYZ) {transformSourceType=TRANSFORMPOSSOURCE_World}
  n58 = Multiply(
   A=Subtract(
    A=ComponentMask(None=@n57) {b=true}
    B=ScalarParameter() {defaultValue=125, group=Liquid Color Scattering, parameterName=Liquid_OuterColor_TransitionStart}
   ) {constB=125}
   B=ScalarParameter() {defaultValue=0.015, group=Liquid Color Scattering, parameterName=Liquid_OuterColor_TransitionScale}
  ) {constB=0.015}
  n59 = Multiply(
   A=LinearInterpolate(
    A=MaterialFunctionCall() {materialFunction=&CameraDirectionVector}.Result
    B=@n49
    Alpha=ScalarParameter() {group=Liquid Color Scattering, parameterName=Use Sun Direction}
   )
   B=ScalarParameter() {defaultValue=1, group=Liquid Color Scattering, parameterName=Liquid_DirectionVectorFlip}
  )
  n60 = Normalize(VectorInput=Transform(None=ComponentMask(None=@n59) {b=true, g=true, r=true}) {transformSourceType=TRANSFORMSOURCE_World, transformType=TRANSFORM_Local})
  n61 = MaterialFunctionCall(Float3=Multiply(A=ComponentMask(None=@n60) {b=true, g=true, r=true}) {constB=0.4886}) {materialFunction=&BreakOutFloat3Components}
  n62 = DotProduct(
   A=MaterialFunctionCall(X=Constant() {r=0.2821}, Y=@n61.B, Z=@n61.B, A=@n61.B) {materialFunction=&MakeFloat4}.Result
   B=Constant4Vector() {constant=<rgba 2 0.5 0.5 0.5>}.RGBA
  )
  n63 = OneMinus(None=MaterialFunctionCall(
    Input=Saturate(None=@n62)
    Input Low=ScalarParameter() {group=Liquid Color Scattering, parameterName=Liquid_SHContrast_LowValue}
    Input High=ScalarParameter() {defaultValue=1, group=Liquid Color Scattering, parameterName=Liquid_SHContrast_HighValue}
    Target Low=Constant()
    Target High=Constant() {r=1}
   ) {materialFunction=&RemapValueRange}.Result)
  n64 = VertexInterpolator(VS=Power(Base=Saturate(None=@n63), Exp=ScalarParameter() {defaultValue=1, group=Liquid Color Scattering, parameterName=Liquid_SHPow}))
  n65 = OneMinus(None=ComponentMask(None=TextureCoordinate() {coordinateIndex=1}) {r=true})
  n66 = QualitySwitch(Default=Add(A=@n64.PS, B=Multiply(A=@n65) {constB=0.0333}), Low=@n64.PS)
  n67 = LinearInterpolate(
   A=VectorParameter() {group=Liquid Color Scattering, parameterName=Liquid_InnerColor, defaultValue=<rgba 0 0.0892 0.25 1>}.RGB
   B=LinearInterpolate(
    A=@n56.RGB
    B=VectorParameter() {group=Liquid Color Scattering, parameterName=Liquid_OuterColorB, defaultValue=<rgba 0 0.5903 0.75 0>}.RGB
    Alpha=Saturate(None=@n58)
   )
   Alpha=ShadingPathSwitch(Default=@n66, Deferred=@n66, Forward=@n64.PS, Mobile=@n64.PS)
  )
  n68 = VectorParameter() {group=Liquid Color Scattering, parameterName=Liquid_InnerScatteringColor, defaultValue=<rgba 0.0877 0.0586 0.2031 1>}
  n69 = Max(A=@n52) {constB=0}
  n70 = ScalarParameter() {defaultValue=5, group=Liquid Color Scattering, parameterName=Liquid_InnerScatteringPower}
  n71 = Saturate(None=Multiply(
    A=Power(Base=@n69, Exp=@n70)
    B=ScalarParameter() {defaultValue=0.4, group=Liquid Color Scattering, parameterName=Liquid_InnerScatteringScale}
   ) {constB=0.4})
  n72 = Multiply(
   A=MaterialFunctionCall(
    Value_Morning=ScalarParameter() {defaultValue=0.66, group=Liquid Color Scattering, parameterName=Liquid_TimeofDayScale_Morning}
    Value_Day=ScalarParameter() {defaultValue=1, group=Liquid Color Scattering, parameterName=Liquid_TimeofDayScale_Day}
    Value_Evening=ScalarParameter() {defaultValue=0.66, group=Liquid Color Scattering, parameterName=Liquid_TimeofDayScale_Evening}
    Value_Night=ScalarParameter() {defaultValue=0.5, group=Liquid Color Scattering, parameterName=Liquid_TimeofDayScale_Night}
   ) {materialFunction=&Scalar_Time_of_Day}.Result
   B=Multiply(A=LinearInterpolate(A=@n46, B=@n47.RGB, Alpha=@n55), B=LinearInterpolate(A=@n67, B=@n68.RGB, Alpha=@n71))
  )
  n73 = LinearInterpolate(
   A=VectorParameter() {group=Cube Highlight, parameterName=HilightColor, defaultValue=<rgba 0 0.05 0.6 1>}.RGB
   B=@n72
   Alpha=ScalarParameter() {group=Controls, parameterName=Use Advanced Liquid Color Scattering}
  )
  n74 = MaterialFunctionCall!(
   UVs
   Mult=ScalarParameter() {defaultValue=2, group=Opacity, parameterName=EdgeMult}
   Exp=ScalarParameter() {defaultValue=1.6, group=Opacity, parameterName=EdgeExp}
  ) {materialFunction=/Game/Effects/Fort_Effects/MaterialFunctions/MF_SquareEdgeMask}
  n75 = TextureCoordinate()
  n76 = Add(A=ComponentMask(None=OneMinus(None=@n75)) {r=true}) {constB=-0.3}
  n77 = Add(A=Multiply(A=@n74.SquareMask), B=Saturate(None=Multiply(A=@n76) {constB=0.5}))
  n78 = ScalarParameter() {defaultValue=8, group=Opacity, parameterName=FrontEdgeMult}
  n79 = ScalarParameter() {defaultValue=3, group=Opacity, parameterName=FrontEdgeExp}
  n80 = MaterialFunctionCall(
   Mult=ScalarParameter() {defaultValue=8, group=Opacity, parameterName=EndEdgeMult}
   Exp=ScalarParameter() {defaultValue=3, group=Opacity, parameterName=EndEdgeExp}
  ) {materialFunction=&MF_SquareEdgeMask}
  n81 = Add(
   A=OneMinus(None=MaterialFunctionCall(Mult=@n78, Exp=@n79) {materialFunction=&MF_SquareEdgeMask}.LeftMask)
   B=LinearInterpolate(B=OneMinus(None=@n80.RightMask), Alpha=ScalarParameter() {group=Opacity, parameterName=UseEndBreakUpLiquidShapeMask, sliderMax=1})
  )
  n82 = MaterialFunctionCall(In=@n27.G, Contrast=ScalarParameter() {defaultValue=0.5, group=Alpha Mask, parameterName=Alpha Texture Contrast}) {materialFunction=&CheapContrast}
  n83 = Multiply(
   A=LinearInterpolate(A=@n74.RightMask, B=Saturate(None=@n77), Alpha=ScalarParameter() {group=Opacity, parameterName=EdgeMaskShape})
   B=Multiply(A=LinearInterpolate(B=@n7.R, Alpha=@n81) {constA=1}, B=Add(A=@n29, B=@n82.Result))
  )
  n84 = OneMinus(None=Power(Base=@n83) {constExponent=0.3})
  n85 = VectorParameter() {group=Color, parameterName=Liquid Color A, defaultValue=<rgba 0 0.3333 0.2429 1>}
  n86 = ScalarParameter() {group=Liquid Color Scattering, parameterName=Use Advanced Liquid Color Scattering}
  n87 = VectorParameter() {group=Color, parameterName=Liquid Color B, defaultValue=<rgba 0 0.1087 0.1667 1>}
  n88 = Power(Base=@n28, Exp=ScalarParameter() {defaultValue=2, group=Color, parameterName=ColorExp})
  n89 = ScalarParameter() {defaultValue=2, group=Color, parameterName=ColorMult}
  n90 = LinearInterpolate(
   A=LinearInterpolate(A=@n85.RGB, B=Multiply(A=@n85.RGB, B=@n72), Alpha=@n86)
   B=LinearInterpolate(A=@n87.RGB, B=Multiply(A=@n87.RGB, B=@n72), Alpha=@n86)
   Alpha=Saturate(None=Multiply(A=@n88, B=@n89))
  )
  n91 = Add(A=@n90, B=Multiply(A=Saturate(None=@n84), B=@n90))
  n92 = Add(
   A=@n91
   B=Multiply(A=Multiply(A=@n45.Result, B=@n73) {constA=1}, B=ScalarParameter() {defaultValue=3, group=Cube Highlight, parameterName=HiLightBrightness})
  )
  n93 = Desaturation(None=@n91, Fraction=ScalarParameter() {defaultValue=-3.5, group=Color, parameterName=DesaturationAmount_LowMode})
  n94 = VectorParameter() {group=Color, parameterName=Liquid Color B, defaultValue=<rgba 0.0019 0.0606 0.125 1>}
  n95 = VectorParameter() {group=Color, parameterName=Liquid Color A, defaultValue=<rgba 0 0.3333 0.2429 1>}
  n96 = TextureCoordinate() {coordinateIndex=1, uTiling=0.5, vTiling=0.5}
  n97 = Multiply(
   A=@n96
   B=ComponentMask(None=VectorParameter() {group=Liquid Trail, parameterName=Bubbles_UV_Scale, defaultValue=<rgba 1 1 0 0>}.RGB) {g=true, r=true}
  )
  n98 = Clamp(None=TextureSample(UVs=Panner(Coordinate=@n97) {speedX=0.0625}) {texture=/Game/Effects/Fort_Effects/Textures/Tile/T_LaserTrail_Dots}.R)
  n99 = ScalarParameter() {defaultValue=0.25, group=Liquid Trail, parameterName=Bubble Intensity}
  n100 = Clamp(None=Add(
    A=TextureSample(UVs=@n96) {samplerSource=SSM_Wrap_WorldGroupSettings, texture=/Game/Effects/Fort_Effects/Textures/Tile/T_Plasma_BallCore}.R
    B=Multiply(A=@n98, B=@n99)
   ))
  n101 = LinearInterpolate(
   A=ShadingPathSwitch(Default=QualitySwitch(Default=@n92, Low=@n93), Forward=@n93, Mobile=@n93)
   B=Multiply(A=LinearInterpolate(A=@n94.RGB, B=@n95.RGB, Alpha=@n100), B=@n72)
   Alpha=ScalarParameter() {group=Controls, parameterName=Use Goop or Liquid}
  )
  n102 = ParticleColor()
  n103 = Multiply(
   A=LinearInterpolate(A=@n101, B=Multiply(A=@n101, B=@n102.RGB), Alpha=ScalarParameter() {group=Controls, parameterName=Multiply by Particle Color})
   B=MaterialFunctionCall(
    Value_Morning=ScalarParameter() {defaultValue=0.5, group=Color, parameterName=Glow_Value_Morning}
    Value_Day=ScalarParameter() {defaultValue=1, group=Color, parameterName=Glow_Value_Day}
    Value_Evening=ScalarParameter() {defaultValue=0.5, group=Color, parameterName=Glow_Value_Evening}
    Value_Night=ScalarParameter() {defaultValue=0.5, group=Color, parameterName=Glow_Value_Night}
   ) {materialFunction=&Scalar_Time_of_Day}.Result
  )
  MaterialFunctionCall(Emissive=@n103) {materialFunction=&MF_FX_EmissiveAdjustments}
  n104 = ScalarParameter() {defaultValue=0.05, group=Opacity, parameterName=ClipRange}
  n105 = Multiply(
   A=ScalarParameter() {defaultValue=0.9, group=Opacity, parameterName=ClipAmnt}
   B=DynamicParameter() {defaultValue=<rgba 1 1 0 1>, paramNames [rand offset, ClipAmnt, WPO Intensity, Param4]}.ClipAmnt
  )
  n106 = Subtract(A=@n83, B=LinearInterpolate(A=Multiply(A=@n104) {constB=-1}, Alpha=OneMinus(None=@n105) {desc=Inverse Alpha}))
  n107 = Multiply(A=Saturate(None=Divide(A=@n106, B=@n104))) {constB=0.9}
  n108 = Panner(Coordinate=MaterialFunctionCall(XY=TextureCoordinate()) {materialFunction=&Swizzle}.YX) {speedX=0.25}
  n109 = TextureCoordinate()
  n110 = Add(
   A=TextureSample(UVs=@n108) {samplerSource=SSM_Wrap_WorldGroupSettings, texture=/Game/Effects/Fort_Effects/Textures/Tile/T_Plasma_BallEdge}.RGB
   B=OneMinus(None=ComponentMask(None=@n109) {r=true})
  )
  n111 = ScalarParameter() {defaultValue=4, group=Liquid Trail, parameterName=Liquid Trail Edge Contrast}
  n112 = ScalarParameter() {defaultValue=1, desc="0 fully softens edges, 1 prevents softening", group=Liquid Trail, parameterName=SoftenEdge, sliderMax=1}
  n113 = Add(A=TextureSample(UVs=@n108) {texture=/Game/Global/Textures/Default/Masks_and_Gradients/T_Soft_Ramp_Vert_03}.RGB, B=@n112)
  n114 = TextureCoordinate() {coordinateIndex=1, uTiling=0.5, vTiling=0.5}
  n115 = ComponentMask(None=VectorParameter() {group=Liquid Trail, parameterName=Liquid Trail Bubbles_UV_Scale, defaultValue=<rgba 1 1 0 0>}.RGB) {g=true, r=true}
  n116 = TextureSample(UVs=Panner(Coordinate=Multiply(A=@n114, B=@n115)) {speedX=0.0625}) {texture=&T_LaserTrail_Dots}
  n117 = Add(
   A=TextureSample(UVs=@n114) {samplerSource=SSM_Wrap_WorldGroupSettings, texture=&T_Plasma_BallCore}.R
   B=Multiply(A=Clamp(None=@n116.R), B=ScalarParameter() {defaultValue=0.25, group=Liquid Trail, parameterName=Liquid Trail Bubble Intensity})
  )
  n118 = Multiply(
   A=Multiply(A=Power(Base=@n110, Exp=@n111), B=Saturate(None=@n113))
   B=Multiply(A=Clamp(None=@n117), B=ScalarParameter() {defaultValue=2, group=Liquid Trail, parameterName=Liquid Trail Alpha Multiplier}) {constA=1, constB=10}
  )
  n119 = ScalarParameter() {group=Controls, parameterName="Use Goop or Liquid Trail "}
  n120 = Power(
   Base=MaterialFunctionCall(Distance=ScalarParameter() {defaultValue=90, group=Opacity, parameterName=NearFade}) {materialFunction=&Near_Camera_Fade}.Result
   Exp=ScalarParameter() {defaultValue=5, group=Opacity, parameterName=NearFadeExp}
  )
  n121 = Multiply(A=Multiply(A=LinearInterpolate(A=@n107, B=@n118, Alpha=@n119), B=Saturate(None=@n120)), B=ParticleColor().A)
  n122 = DepthFade(Opacity=@n121, FadeDistance=ScalarParameter() {defaultValue=20, group=DepthDade, parameterName=DepthFadeDistance, sortPriority=1})
  Reroute(None=LinearInterpolate(
    A=@n121
    B=ShadingPathSwitch(Default=@n122, Deferred=@n122, Mobile=@n121) {desc=Migrated from FeatureLevelSwitch}
    Alpha=ScalarParameter() {group=DepthDade, parameterName=UseDepthFade}
   ))
  Reroute(None=@n121)
  Reroute(None=@n103)
  n123 = TextureCoordinate()
  n124 = ComponentMask(None=VectorParameter() {group=WPO, parameterName=WPO UV Scale, defaultValue=<rgba 1 1 0 0>}.RGB) {g=true, r=true}
  n125 = Time() {bOverride_Period=true, period=64}
  n126 = DynamicParameter() {defaultValue=<rgba 1 1 0 1>, paramNames [rand offset, ClipAmnt, WPO Intensity, Param4]}
  n127 = VectorParameter() {group=WPO, parameterName=WPO Panner Speed}
  n128 = TextureSampleParameter2D(UVs=Panner(Coordinate=Multiply(A=@n123, B=@n124), Time=Add(A=@n125, B=@n126.rand offset), Speed=ComponentMask(None=@n127.RGB) {g=true, r=true}) {speedX=-0.1, speedY=-0.01}) {group=WPO, parameterName=WPO Texture, samplerType=SAMPLERTYPE_Grayscale, texture=&LowResBlurredNoise}
  Transform(None=Multiply(A=@n128.RGB, B=@n126.WPO Intensity))
 ]
}
